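# Sound cue for the impact sound of the Gnome cosmetic pickaxe.
# A fixed short delay feeds a modulator that slightly raises pitch at random, over an equally weighted random pick of six impact waves.
# It uses pickaxe impact attenuation, property sheet and impacts sound class, and primes on load.
# A melee weapon concurrency set is assigned, alongside per-owner override values that stop the oldest sound.
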SoundCue /BRCosmetics/Sounds/PickAxes/Gnome/Pickaxes_Gnome_Impact
root=Delay_1
nodes [
 {id=Delay_1, children [Modulator_0], properties {delayMax=0.1, delayMin=0.1}}
 {id=Modulator_0, children [Random_2], properties {pitchMax=1.1, pitchMin=1}}
 {
  id=Random_2
  children [WavePlayer_10, WavePlayer_11, WavePlayer_12, WavePlayer_13, WavePlayer_14, WavePlayer_15]
  properties {weights [1,1,1,1,1,1]}
 }
 {id=WavePlayer_10, properties {soundWaveAssetPtr=&/Pickaxes_Gnome_Impact_01}}
 {id=WavePlayer_11, properties {soundWaveAssetPtr=&/Pickaxes_Gnome_Impact_02}}
 {id=WavePlayer_12, properties {soundWaveAssetPtr=&/Pickaxes_Gnome_Impact_03}}
 {id=WavePlayer_13, properties {soundWaveAssetPtr=/Game/Athena/Sounds/Weapons/PickAxes/Gnome/Pickaxes_Gnome_Impact_04}}
 {id=WavePlayer_14, properties {soundWaveAssetPtr=/Game/Athena/Sounds/Weapons/PickAxes/Gnome/Pickaxes_Gnome_Impact_05}}
 {id=WavePlayer_15, properties {soundWaveAssetPtr=/Game/Athena/Sounds/Weapons/PickAxes/Gnome/Pickaxes_Gnome_Impact_06}}
]
properties {
 attenuationSettings=/Game/Sounds/Attenuation/Fort_Pickaxe_Impact_Att
 audioPropertiesSheet=/BRCosmetics/Sounds/Settings/PropertySheets/Pickaxe/APS_Cosmetic_Pickaxe_Impact
 bPrimeOnLoad=true
 maxDistance=4300
 soundClassObject=&Fort_Impacts
 volumeMultiplier=0.9
 concurrencyOverrides {bLimitToOwner=true, maxCount=2, resolutionRule=StopOldest}
 concurrencySet [&Player_Melee_Weapon_Conc]
}
